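# Niagara system for the "Pops" vehicle boost cosmetic, emitting colored exhaust effects from the car's boost sockets.
# BaseMist layers mist, fire and sparks, alongside ribbon trails, short sprite bursts, and falling shards killed on collision, with a plane for frontend preview.
# Emitters spread particles across filtered skeletal-mesh sockets by port count, using spawn groups to vary motion, size and renderer.
# Per-emitter painted-variant tables set colors, with secondary R1/R2 palettes for fire and sparks.
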NiagaraSystem /VehicleCosmetics/Boosts/Boost_Pops/FX/NS_Boost_Pops
dependencies {
 usedDataInterfaces [
  &NiagaraDataInterfaceCurve
  &NiagaraDataInterfaceArrayFloat3
  &NiagaraDataInterfaceArrayFloat
  &NiagaraDataInterfaceVectorCurve
  &NiagaraDataInterfaceSpriteRendererInfo
  &NiagaraDataInterfaceArrayColor
  &NiagaraDataInterfaceSkeletalMesh
 ]
 usedDynamicInputs []
 usedModules [
  &self:SystemSpawnScript.SystemScriptSource.SystemScriptGraph.NiagaraNodeAssignment_0.SetVariables_#1
  &SystemState
 ]
}
emitters {
 BaseMist {
  emitterUpdateScript [
   &EmitterState()
   &SpawnRate(SpawnGroup=2, SpawnRate=20, Use Spawn Probability=false) {module=SpawnRate002}
   &SpawnRate(SpawnGroup=1, SpawnRate=60, Use Spawn Probability=false) {module=SpawnRate001}
   &VC_FilteredSocketCount()
   &SpawnRate(SpawnRate=40, Use Spawn Probability=false)
  ]
  particleSpawnScript [
   &VC_PaintedVariantsOverride(
    Black=<rgba 0.05 0.05 0.05 0.5>
    BurntSienna=<rgba 0.5104 0.2945 0.2366 0.6>
    Cobalt=<rgba 0.45 0.5848 1 0.4>
    Crimson=<rgba 1 0.6 0.7 0.2>
    DefaultColor=<rgba 1.3 0.975 1.003 0.2>
    ForestGreen=<rgba 0.24 0.8 0.38 0.3>
    Grey=<rgba 0.63 0.6434 0.7 0.4>
    Lime=<rgba 1.032 1.2 0.36 0.5>
    Orange=<rgba 0.887 0.6283 0.5765 0.3>
    Pink=<rgba 0.98 0.7 1 0.3>
    Purple=<rgba 0.4447 0.174 0.58 0.4>
    R1 Black:LinearColor=<rgba 0 0 0 1>
    R1 BurntSienna:LinearColor=<rgba 0.8 0.1067 0 0.7>
    R1 Cobalt:LinearColor=<rgba 0.3 0.3314 2 1>
    R1 Crimson:LinearColor=<rgba 5 0.15 0.15 1>
    R1 DefaultColor=<rgba 2 0.7033 0.5 1>
    R1 ForestGreen:LinearColor=<rgba 0.08 0.8 0.1164 1>
    R1 Grey:LinearColor=<rgba 0.2031 0.2031 0.2031 0.6>
    R1 Lime:LinearColor=<rgba 1.2208 2 0.13 1>
    R1 Orange:LinearColor=<rgba 1.5 0.4425 0.15 1>
    R1 Pink:LinearColor=<rgba 2 0.4 1.4182 1>
    R1 Purple:LinearColor=<rgba 0.8453 0.32 1.6 1>
    R1 Saffron:LinearColor=<rgba 5 3.8333 0 1>
    R1 SkyBlue:LinearColor=<rgba 0.45 1.1487 1.5 1>
    R1 TitaniumWhite:LinearColor=<rgba 2 2 2 1>
    R2 Black:LinearColor=<rgba 3 3 3 1>
    R2 BurntSienna:LinearColor=<rgba 3 0.9263 0.3698 1>
    R2 Cobalt:LinearColor=<rgba 0.75 0.8286 5 1>
    R2 Crimson:LinearColor=<rgba 5 0 0.1355 1>
    R2 DefaultColor:LinearColor=<rgba 5 2.4065 2 1>
    R2 ForestGreen:LinearColor=<rgba 0 5 0.2526 1>
    R2 Grey:LinearColor=<rgba 4 4 4 1>
    R2 Lime:LinearColor=<rgba 2.5 5 0 1>
    R2 Orange:LinearColor=<rgba 4 2.1767 1.2 1>
    R2 Pink:LinearColor=<rgba 5 2 3.9091 1>
    R2 Purple:LinearColor=<rgba 4.1041 0 10 1>
    R2 Saffron:LinearColor=<rgba 5 3.8333 0 1>
    R2 SkyBlue:LinearColor=<rgba 1 3.6615 5 1>
    R2 TitaniumWhite:LinearColor=<rgba 5 5 5 1>
    Saffron=<rgba 1.2 1.14 0.48 0.3>
    SkyBlue=<rgba 0.6 0.9 1 0.3>
    TitaniumWhite=<rgba 1.5 1.5 1.5 0.2>
    Use R1 Painted Overrides=true
    Use R2 Painted Overrides=true
   )
   &InitializeParticle(
    Color Mode=::NewEnumerator0
    Lifetime=/VehicleCosmetics/FX/Niagara/Modules/MergeEmitters/VC_FloatBySpawnGroup(
     SpawnGroup or ExecIndex:NiagaraInt32=0
     CustomInt:NiagaraInt32=0
     MaximumSpawnGroups:NiagaraInt32=3
     Float SG 0:NiagaraFloat=&RandomRangeFloat(Minimum=0.3, Maximum=0.6)
     Float SG 1:NiagaraFloat=&RandomRangeFloat(Minimum=0.15, Maximum=0.3)
     Float SG 2:NiagaraFloat=&RandomRangeFloat(Minimum=0.15, Maximum=0.3)
     Float SG 3:NiagaraFloat=&FloatFromCurve(FloatCurve=DataInterface() {propertyValues {Curve {keys [<at 0 8>,<at 1 8>]}}})
     Float SG 4:NiagaraFloat=0
     Float SG 5:NiagaraFloat=0
    )
    Sprite Rotation Mode=::NewEnumerator2
    Sprite Size Mode=::NewEnumerator3
    Uniform Sprite Size=&RandomRangeFloat(Minimum=20, Maximum=30)
    UsePositionOffset=false
    Write Material Random=true
   )
   &PartitionParticles(Number Of Partitions=$Emitter.PortCount)
   &SkeletalMeshLocation(
    Bone / Socket Index=$Particles.PartitionParticles.Partition
    Bone / Socket Sampling Mode=::NewEnumerator5
    MaskSpawnGroup=false
    Mesh Sampling Type=::NewEnumerator6
    Sampled Position Offset=<-70 0 0>
    Skeletal Mesh=$System.SkeletalMesh
   )
   &AddVelocity(Velocity=&SelectVectorFromArray(
     Direct Array Index=&SpawnGroup()
     Vector Selection Array=DataInterface() {propertyValues {FloatData [<-800 0 0>, <-300 0 0>, <-800 0 0>]}}
    ))
   /SetVariables(Particles.VisibilityTag=&SpawnGroup()) {isSetParameters=true}
   &ShapeLocation(
    MaskSpawnGroup=false
    Override Seed=false
    Sphere Radius=&SelectFloatFromArray(Direct Array Index=&SpawnGroup(), Float Selection Array=DataInterface() {propertyValues {FloatData [30,0,50]}})
    Sphere Surface Distribution=1
    UseOwnerScale=true
   )
  ]
  particleUpdateScript [
   &ParticleState()
   &ScaleSpriteSize(Uniform Curve Sprite Scale=DataInterface() {propertyValues {Curve {keys [<at 0 0.7 RCIM_Cubic>, {arriveTangent=2.8, interpMode=RCIM_Cubic, leaveTangent=2.8, time=0.2, value=3}, <at 1 3.5 RCIM_Cubic>]}}})
   &InheritSourceMovement(
    Applied Position Scale=&VectorFromCurve(VectorCurve=DataInterface() {propertyValues {
       XCurve {keys [<at 0 1 RCIM_Cubic>,<at 1 0 RCIM_Cubic>]}
       YCurve {keys [<at 0 1 RCIM_Cubic>,<at 1 0>]}
       ZCurve {keys [<at 0 1 RCIM_Cubic>,<at 1 0>]}
      }})
    ApplyForce=false
    ApplyPosition=true
    ApplyPositionFalloff=false
   )
   &SolveForcesAndVelocity(Clamp Velocity=false, Limit Acceleration=false)
   &ScaleColor(
    Scale Alpha=&FloatFromCurve(FloatCurve=DataInterface() {propertyValues {Curve {
        defaultValue=1
        keys [
         {arriveTangent=9.0725, interpMode=RCIM_Cubic, time=0, value=0}
         {arriveTangent=1.25, leaveTangent=1.25, time=0.1603, value=2}
         {arriveTangent=-2.6365, interpMode=RCIM_Cubic, leaveTangent=-2.6365, tangentMode=RCTM_User, time=0.5106, value=1}
         {arriveTangent=-0.0405, interpMode=RCIM_Cubic, leaveTangent=-0.0405, tangentMode=RCTM_User, time=1, value=0}
        ]
       }}})
    ScaleA=true
    ScaleRGB=true
   )
   /SetVariables(
    Particles.FireColor:LinearColor=&MultiplyLinearColor(
     Linear Color=$Particles.R1 Color
     Scale Factor=&MakeVector4FromVectorAndFloat(
      "Vector (RGB)"=<1 1 1>
      "Float (Alpha)"=&FloatFromCurve(FloatCurve=DataInterface() {propertyValues {Curve {keys [<at 0 1>,<at 1 0>]}}})
     )
    )
    Particles.FireSize:Vector2f=&Vector2DFromFloat(Value=&RandomRangeFloat(Minimum=60, Maximum=120))
    Particles.SparksColor:LinearColor=$Particles.R2 Color
    Particles.SparksSieze:Vector2f=&Vector2DFromFloat(Value=&FloatFromCurve(
      FloatCurve=DataInterface() {propertyValues {Curve {keys [<at 0 0.7 RCIM_Cubic>, {arriveTangent=4.3, interpMode=RCIM_Cubic, leaveTangent=4.3, time=0.2, value=3}, <at 1 5 RCIM_Cubic>]}}}
      Scale Curve=&RandomRangeFloat(Minimum=25, Maximum=35)
     ))
   ) {isSetParameters=true}
   &SubUVAnimation(
    "Sprite Renderer "=DataInterface() {propertyValues {SpriteRenderer=&self:BaseMist_0.NiagaraSpriteRendererProperties_5}}
    UseEndFrame=false
    UseStartFrame=false
   )
   &SubUVAnimation(
    "Sprite Renderer "=DataInterface() {propertyValues {SpriteRenderer=&self:BaseMist_0.NiagaraSpriteRendererProperties_0}}
    Start Frame Range Override=40
    UseEndFrame=true
    UseStartFrame=true
   ) {module=SubUVAnimation001}
  ]
  rendererClasses [&NiagaraSpriteRendererProperties]
  renderers [
   {rendererClass=&NiagaraSpriteRendererProperties}
   {rendererClass=&NiagaraSpriteRendererProperties, rendererIndex=1}
   {rendererClass=&NiagaraSpriteRendererProperties, rendererIndex=2}
  ]
 }
 NE_VC_Boost_Ribbon_Parent {
  emitterUpdateScript [
   &EmitterState()
   &VC_FilteredSocketCount()
   &VC_RateQualityHandler()
   &VC_SpawnRibbon_v1(Number of Ribbons=$Emitter.PortCount, SpawnRate=$Emitter.Spawn Rate)
  ]
  particleSpawnScript [
   &VC_PaintedVariantsOverride(
    Black=<rgba 0 0 0 50>
    BurntSienna=<rgba 3 1.3443 0.9 20>
    Cobalt=<rgba 1.75 1 10 10>
    Crimson=<rgba 7 2.1 2.2328 10>
    DefaultColor=<rgba 30 14.439 12 1>
    ForestGreen=<rgba 1 10 1.4547 10>
    Grey=<rgba 10 10 10 5>
    Lime=<rgba 6 10 2 10>
    Orange=<rgba 10 4.7904 2 10>
    Pink=<rgba 8 2.4 5.9636 10>
    Purple=<rgba 3.6937 0 9 25>
    Saffron=<rgba 10 8.3667 3 10>
    SkyBlue=<rgba 1 6.9885 10 10>
    TitaniumWhite=<rgba 30 30 30 1>
   )
   &VC_InitializeRibbon_v1(
    Color=$Particles.Color
    Custom Offset=true
    Enable Skeletal Position=true
    Lifetime=1
    Local Vector Offset=<-70 0 5>
    Offset Amount=1
    Ribbon Width=50
   )
   &AddVelocity(Velocity=<-1500 0 0>) {module=AddVelocity001}
  ]
  particleUpdateScript [
   &ParticleState()
   &InheritSourceMovement(ApplyForce=false, ApplyPosition=true, ApplyPositionFalloff=false)
   &SolveForcesAndVelocity(Clamp Velocity=false, Limit Acceleration=false)
   &VC_UpdateRibbon_v1(Enable Skeletal Position=true)
   &ScaleColor(
    Scale Alpha=&FloatFromCurve(FloatCurve=DataInterface() {propertyValues {Curve {
        defaultValue=0
        keys [
         {arriveTangent=9.0725, interpMode=RCIM_Cubic, leaveTangent=9.0725, tangentMode=RCTM_User, time=0, value=0}
         {arriveTangent=-2.5249, interpMode=RCIM_Cubic, leaveTangent=-2.5249, tangentMode=RCTM_User, time=0.1408, value=0.9847}
         {arriveTangent=-0.0405, interpMode=RCIM_Cubic, leaveTangent=-0.0405, tangentMode=RCTM_User, time=1, value=0}
        ]
       }}})
    Scale RGB=&VectorFromFloat(Value=&FloatFromCurve(FloatCurve=DataInterface() {propertyValues {Curve {
         defaultValue=0
         keys [
          {arriveTangent=9.0725, interpMode=RCIM_Cubic, leaveTangent=9.0725, tangentMode=RCTM_User, time=0, value=0}
          {arriveTangent=-2.5249, interpMode=RCIM_Cubic, leaveTangent=-2.5249, tangentMode=RCTM_User, time=0.1408, value=0.9847}
          {arriveTangent=-0.0405, interpMode=RCIM_Cubic, leaveTangent=-0.0405, tangentMode=RCTM_User, time=1, value=0}
         ]
        }}}))
    ScaleA=true
    ScaleRGB=true
   )
  ]
  rendererClasses [&NiagaraRibbonRendererProperties]
  renderers [{rendererClass=&NiagaraRibbonRendererProperties}]
 }
 NE_VC_Boost_Sprite_Parent001 {
  emitterUpdateScript [
   &EmitterState(Inactive Response=::NewEnumerator1, Life Cycle Mode=::NewEnumerator1, Loop Duration=0.2, UseLoopDelay=false)
   &VC_FilteredSocketCount()
   &SpawnRate(SpawnRate=20, Use Spawn Probability=false)
   &SpawnRate(SpawnGroup=1, SpawnRate=80, Use Spawn Probability=false) {module=SpawnRate001}
   &SpawnRate(SpawnGroup=2, SpawnRate=250, Use Spawn Probability=false) {module=SpawnRate002}
  ]
  particleSpawnScript [
   /SetVariables(Particles.RenderTag:NiagaraInt32=&SpawnGroup()) {isSetParameters=true}
   &VC_PaintedVariantsOverride(
    Black=<rgba 3 3 3 1>
    BurntSienna=<rgba 3 1.1836 0.2472 1>
    Cobalt=<rgba 1.4531 1.5927 9 1>
    Crimson=<rgba 9 1.7813 1.9769 1>
    DefaultColor=<rgba 50 15.4201 10 1>
    ForestGreen=<rgba 0.8333 4 0.9933 1.5>
    Grey=<rgba 0.1042 0.1042 0.1042 1>
    Lime=<rgba 4.5 9 0 1>
    Orange=<rgba 4 1.2888 0.4004 1>
    Pink=<rgba 9 1.0781 6.1193 1>
    Purple=<rgba 1.9319 0 12 1>
    Saffron=<rgba 3 2.4349 0.5781 1>
    SkyBlue=<rgba 1.8281 6.6002 9 1>
    TitaniumWhite=<rgba 9 9 9 1>
   )
   &InitializeParticle(
    Color Mode=::NewEnumerator0
    Lifetime=&VC_FloatBySpawnGroup(
     SpawnGroup or ExecIndex:NiagaraInt32=0
     CustomInt:NiagaraInt32=0
     MaximumSpawnGroups:NiagaraInt32=4
     Float SG 0:NiagaraFloat=&RandomRangeFloat(Minimum=0.3, Maximum=$Emitter.CurrentLoopDuration)
     Float SG 1:NiagaraFloat=&RandomRangeFloat(Minimum=0.1, Maximum=0.15)
     Float SG 2:NiagaraFloat=&RandomRangeFloat(Minimum=0.18, Maximum=0.24)
     Float SG 3:NiagaraFloat=1
     Float SG 4:NiagaraFloat=0
     Float SG 5:NiagaraFloat=0
    )
    Sprite Rotation Mode=::NewEnumerator2
    Sprite Size Mode=::NewEnumerator3
    Uniform Sprite Size=&VC_FloatBySpawnGroup(
     SpawnGroup or ExecIndex:NiagaraInt32=0
     CustomInt:NiagaraInt32=0
     MaximumSpawnGroups:NiagaraInt32=3
     Float SG 0:NiagaraFloat=20
     Float SG 1:NiagaraFloat=&RandomRangeFloat(Minimum=3, Maximum=5)
     Float SG 2:NiagaraFloat=0
     Float SG 3:NiagaraFloat=0
     Float SG 4:NiagaraFloat=0
     Float SG 5:NiagaraFloat=0
    )
    UsePositionOffset=false
    Write Material Random=true
   )
   &PartitionParticles(Number Of Partitions=$Emitter.PortCount)
   &SkeletalMeshLocation(
    Bone / Socket Index=$Particles.PartitionParticles.Partition
    Bone / Socket Sampling Mode=::NewEnumerator5
    MaskSpawnGroup=false
    Mesh Sampling Type=::NewEnumerator6
    Sampled Position Offset=&AddVectorToPosition(Position=<-70 0 0>, Vector=$User.DTCarOffset)
    Skeletal Mesh=$System.SkeletalMesh
   )
   &AddVelocity(Velocity=/VehicleCosmetics/FX/Niagara/Modules/MergeEmitters/VC_VectorBySpawnGroup(
     SpawnGroup or ExecIndex:NiagaraInt32=0
     CustomInt:NiagaraInt32=0
     MaximumSpawnGroups:NiagaraInt32=4
     Vector Group 0:Vector3f=<0 0 0>
     Vector Group 1:Vector3f=&RandomRangeVector(Minimum=<-300 -300 -300>, Maximum=<300 300 300>)
     Vector Group 2:Vector3f=&RandomRangeVector(Minimum=<-300 -300 -300>, Maximum=<300 300 300>)
     Vector Group 3:Vector3f=<0 0 0>
     Vector Group 4:Vector3f=<0 0 0>
     Vector Group 5:Vector3f=<0 0 0>
    ))
   /SetVariables(Particles.ColorTag1:LinearColor=$Particles.R1 Color, Particles.SizeTag2:Vector2f=<2 8>) {isSetParameters=true}
  ]
  particleUpdateScript [
   &ParticleState()
   &ScaleSpriteSize(Uniform Curve Sprite Scale=DataInterface() {propertyValues {Curve {keys [{tangentMode=RCTM_User, time=0, value=1.5}, {interpMode=RCIM_Cubic, tangentMode=RCTM_Break, time=1, value=3}]}}})
   &SolveForcesAndVelocity(Clamp Velocity=false, Limit Acceleration=false)
   &ScaleColor(
    Scale Alpha=&FloatFromCurve(FloatCurve=DataInterface() {propertyValues {Curve {keys [<at 0 0>,<at 0.1 1>,<at 0.9 1>,<at 1 0>]}}})
    Scale RGB=&VectorFromFloat(Value=10)
    ScaleA=true
    ScaleRGB=true
   )
  ]
  rendererClasses [&NiagaraSpriteRendererProperties]
  renderers [
   {rendererClass=&NiagaraSpriteRendererProperties}
   {rendererClass=&NiagaraSpriteRendererProperties, rendererIndex=1}
  ]
 }
 Shards2001 {
  emitterSpawnScript [/SetVariables(Emitter.ShardsLights:NiagaraFloat=&SelectFloatFromArray(
     Direct Array Index=$System.ColorVariantID
     Float Selection Array=DataInterface() {propertyValues {FloatData [-1,0,-1.3,-0.8,-1,-1,-1,-1,-1,-1,-1,-1,-1,1]}}
    )) {isSetParameters=true}]
  emitterUpdateScript [
   &EmitterState()
   &VC_FilteredSocketCount()
   &SpawnRate(SpawnRate=60, Use Spawn Probability=false)
   &SpawnRate(SpawnGroup=1, SpawnRate=30, Use Spawn Probability=false) {module=SpawnRate001}
  ]
  particleSpawnScript [
   &VC_PaintedVariantsOverride(
    Black=&SelectLinearColorFromArray(
     Array Sampling Mode=::NewEnumerator1
     Direct Array Index=&SpawnGroup()
     Color Selection Array=DataInterface() {propertyValues {ColorData [<rgba 0 0 0 1>, <rgba 3 3 3 1>]}}
     Random Seed=?
    )
    BurntSienna=<rgba 0.513 0.0855 0 1>
    Cobalt=<rgba 2.7 4.275 9 3>
    Crimson=<rgba 9 2.7 2.805 1>
    DefaultColor=<rgba 9 4.05 3.6 1>
    ForestGreen=<rgba 0.9 9 0.9 1>
    Grey=&SelectLinearColorFromArray(
     Array Sampling Mode=::NewEnumerator1
     Direct Array Index=&SpawnGroup()
     Color Selection Array=DataInterface() {propertyValues {ColorData [<rgba 0.1739 0.1739 0.1739 3>, <rgba 0.05 0.05 0.05 10>]}}
     Random Seed=?
    )
    Lime=<rgba 6 9 1.8 1>
    Orange=<rgba 7 1.1667 0 1>
    Pink=<rgba 9 2.7 6.7091 1>
    Purple=<rgba 6.977 0 17 3>
    Saffron=<rgba 9 9 2.7 3>
    SkyBlue=<rgba 2.52 7.92 9 3>
    TitaniumWhite=<rgba 9 9 9 3>
   )
   &InitializeParticle(
    Color=$Particles.Color
    Lifetime=&RandomRangeFloat(
     Minimum=&SelectFloatFromArray(Direct Array Index=&SpawnGroup(), Float Selection Array=DataInterface() {propertyValues {FloatData [0.4,0.2]}})
     Maximum=0.6
    )
    Sprite Rotation Mode=::NewEnumerator2
    Sprite Size Max=<10 20>
    Sprite Size Min=<20 40>
    Sprite Size Mode=::NewEnumerator2
    UsePositionOffset=false
    Write Material Random=true
   )
   &PartitionParticles(Number Of Partitions=$Emitter.PortCount)
   &SkeletalMeshLocation(
    Bone / Socket Index=$Particles.PartitionParticles.Partition
    Bone / Socket Sampling Mode=::NewEnumerator5
    MaskSpawnGroup=false
    Mesh Sampling Type=::NewEnumerator6
    Skeletal Mesh=$System.SkeletalMesh
   )
   &AddVelocity(
    Cone Angle=20
    Cone Axis=<1 0 -0.05>
    Module.UseConeFalloff=false
    Module.UseVelDistribution=false
    Override Seed=false
    Velocity Mode=::NewEnumerator2
    Velocity Speed=&RandomRangeFloat(
     Minimum=&SelectFloatFromArray(
      Direct Array Index=&SpawnGroup()
      Float Selection Array=DataInterface() {propertyValues {FloatData [-1500,-1000]}}
     )
     Maximum=&SelectFloatFromArray(
      Direct Array Index=&SpawnGroup()
      Float Selection Array=DataInterface() {propertyValues {FloatData [-2000,-2000]}}
     )
    )
   )
   &ShapeLocation(MaskSpawnGroup=false, Override Seed=false, Sphere Radius=40, UseOwnerScale=true)
   /SetVariables(Particles.RenderVis:NiagaraInt32=&SpawnGroup()) {isSetParameters=true}
  ]
  particleUpdateScript [
   &ParticleState()
   &ScaleSpriteSize(Uniform Curve Sprite Scale=DataInterface() {propertyValues {Curve {keys [
        <at 0.2114 0.8029 RCIM_Cubic>
        {arriveTangentWeight=0.2491, interpMode=RCIM_Cubic, leaveTangentWeight=0.1993, tangentMode=RCTM_User, time=0.6, value=1}
        <at 1 0 RCIM_Cubic>
       ]}}})
   &AccelerationForce(Acceleration=<0 0 -1500>)
   &ScaleColor(Scale RGB=&Multiply_VectorByFloat(Vector=<1 1 1>, Float=7), ScaleA=true, ScaleRGB=true) {enabled=false, module=ScaleColor001}
   &InheritSourceMovement(
    Applied Position Scale=&VectorFromCurve(VectorCurve=DataInterface() {propertyValues {
       XCurve {keys [<at 0 1 RCIM_Cubic>,<at 1 0 RCIM_Cubic>]}
       YCurve {keys [<at 0 1 RCIM_Cubic>,<at 1 0>]}
       ZCurve {keys [<at 0 1 RCIM_Cubic>,<at 1 0>]}
      }})
    ApplyForce=false
    ApplyPosition=true
    ApplyPositionFalloff=false
   )
   &Collision(
    Age Colliding Particles=false
    Analytical Collision Plane Position 2:Vector3f=&MakeCustomVectorFromBool(Bool=$User.bIsFrontendPreview, True Vector=<0 0 -90>, False Vector=<1 0 -100000000>)
    Analytical Collision Plane Space:ENiagaraCoordinateSpace=::Local
    CPU Collision Type=::NewEnumerator3
    KillOnCollision=true
    Module.Control Rest State Via Surface Normal=false
    Module.EnableMaxCollisionCount=false
    Module.Randomize Collision Normal Vector=false
    Restitution=0
   )
   &SolveForcesAndVelocity(Clamp Velocity=false, Limit Acceleration=false)
   &SubUVAnimation(
    "Sprite Renderer "=DataInterface() {propertyValues {SpriteRenderer=&self:Shards2001_0.NiagaraSpriteRendererProperties_4}}
    UseEndFrame=false
    UseStartFrame=false
   )
   &SubUVAnimation(
    "Sprite Renderer "=DataInterface() {propertyValues {SpriteRenderer=&self:Shards2001_0.NiagaraSpriteRendererProperties_5}}
    UseEndFrame=false
    UseStartFrame=false
   ) {module=SubUVAnimation001}
  ]
  rendererClasses [&NiagaraSpriteRendererProperties]
  renderers [
   {rendererClass=&NiagaraSpriteRendererProperties}
   {rendererClass=&NiagaraSpriteRendererProperties, rendererIndex=1}
  ]
 }
}
summary {
 systemName=NS_Boost_Pops
 emitters [
  {emitterName=BaseMist, rendererClasses [&NiagaraSpriteRendererProperties]}
  {emitterName=Shards2001, rendererClasses [&NiagaraSpriteRendererProperties]}
  {emitterName=NE_VC_Boost_Sprite_Parent001, rendererClasses [&NiagaraSpriteRendererProperties]}
  {emitterName=NE_VC_Boost_Ribbon_Parent, rendererClasses [&NiagaraRibbonRendererProperties]}
 ]
 userVariables [
  User.bIsFrontendPreview {type=NiagaraBool}
  User.bIsPainted {type=NiagaraBool}
  User.ColorVariantID {type=NiagaraInt32}
  User.DTCarOffset {type=Vector3f, description="", defaultValue {struct=Vector3f, value=<0 0 0>}}
  User.PaintedColor {type=LinearColor, description="", defaultValue {struct=LinearColor, value=<rgba 1 1 1 1>}}
  User.PaintedScalar {type=NiagaraFloat, defaultValue=1, description=""}
  User.PortCount {type=NiagaraInt32, defaultValue=2}
 ]
}
